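; Weapon Blueprint for the Muster Energy Drink consumable, managing the held bottle's liquid effect and animation state.
; On BeginPlay it stores the owning pawn, activates the liquid Niagara effect on refill gameplay events, and deactivates it when the owner's mesh starts the drink montage.
; Equipping reinitializes the effect, while local unequip stops and rewinds the weapon mesh montage.
; An RPC event hides the bottle mesh.
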(event Custom|RPC_HideBottle
  (Rendering|SetVisibility (Variables|Visual|GetWeaponMesh) false true))

(event EventBeginPlay
  (bind _asfort_pawn (Utilities|Casting|CastToFortPawn (Game|GetInstigator)))
  (bind _output_get (Variables|Default|SetWeaponOwner _asfort_pawn))
  (|CallParentFunction)
  (EventDispatchers|BindEventtoOnMontageStarted (Components|SkeletalMesh|GetAnimInstance (Variables|/Script/Engine.Character|GetMesh _output_get)) (delegate Custom|MontageStarted))
  (Ability|Async|WaitGameplayEventToActor (Variables|Default|GetWeaponOwner) "(TagName=\"Event.Muster.EnergyDrink.Refill\")"
    (:EventReceived
      (Components|Activation|Activate (Variables|Default|GetNS_EnergyDrink_PickupLiquid) false))))

(event Custom|MontageStarted
  (params (Montage :type Anim%20Montage%20Object%20Reference))
  (if (== Montage "/MusterConsumables/Items/EnergyDrink/Animation/EnergyDrink_Montage.EnergyDrink_Montage")
    (Components|Activation|Deactivate (Variables|Default|GetNS_EnergyDrink_PickupLiquid))))

(event FortWeapon|EventOnEquippedAndAttached
  ;[0.000548 0.1541 1] Ensure drink returns to default values when re-equipped
  (Niagara|ReinitializeSystem (Variables|Default|GetNS_EnergyDrink_PickupLiquid)))

(event FortWeapon|EventOnLocalUnEquip
  (bind _weaponmesh (Variables|Visual|GetWeaponMesh))
  (bind _returnvalue (Components|SkeletalMesh|GetAnimInstance _weaponmesh))
  (bind _returnvalue_2 (Animation|Montage|GetCurrentActiveMontage _returnvalue))
  ;[0.000548 0.1541 1] Manually reset animation position when the energy drink is unequipped
  (Animation|Montage|MontageStop _returnvalue 0.0 _returnvalue_2)
  (Animation|Montage|MontageSetPosition _returnvalue _returnvalue_2 0.0))

(fn ConstructionScript
  (|CallParentFunction))
