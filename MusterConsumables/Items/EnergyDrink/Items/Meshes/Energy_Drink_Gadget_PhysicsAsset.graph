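# Physics asset for the Energy Drink gadget mesh in the MusterConsumables content.
# It holds a single body on the Bottle_LOD0 bone, shaped as one narrow capsule fitted along the bottle.
# It defines no constraints, so the bottle collides and simulates as one rigid piece.
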PhysicsAsset /MusterConsumables/Items/EnergyDrink/Items/Meshes/Energy_Drink_Gadget_PhysicsAsset
bodies {Bottle_LOD0 [{shapeType=Capsule, length=21.0266, radius=1.5504, center=<0.004 -0.0016 18.5311>, rotation=<rot 0.0132 -90 -0.0321>}]}
constraints []
